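# Translucent Default Lit master material for anime-style glass, usable on skeletal meshes and mesh deformers.
# A sphere-masked sun half-vector highlight and two sheen bands, shifted by view angle along rotated UVs, add to a Fresnel-blended opacity range.
# Fresnel also blends two tint colors, and a scene-versus-pixel depth fade scales both that tint and a scene color sample.
# Sun direction comes from the FortniteMaterialParameters collection.
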Material /BRCosmetics/Packages/Fortress_SharedMaterials/Anime_Master_Material/M_MASTER_AnimeGlass
graph {
 properties {
  bUsedWithMeshDeformer=true
  bUsedWithSkeletalMesh=true
  blendMode=BLEND_Translucent
  pixelDepthOffsetMode=PDOM_Legacy
  usedShadingModels=Default Lit
 }
 nodes [
  n1 = CameraVectorWS()
  n2 = CollectionParameter() {
   collection=&FortniteMaterialParameters
   parameterName=SunlightDirectionalVector
  }
  n3 = ComponentMask(None=@n2) {b=true, g=true, r=true}
  n4 = PixelNormalWS()
  n5 = DotProduct(A=Normalize(VectorInput=Add(A=@n1, B=@n3)), B=@n4)
  n6 = ScalarParameter() {defaultValue=1, parameterName=spheremaskb}
  n7 = ScalarParameter() {defaultValue=0.005, parameterName=SpecPower}
  n8 = ScalarParameter() {defaultValue=1, parameterName=Sharpness}
  n9 = Multiply(
   A=DotProduct(
    A=MaterialFunctionCall() {materialFunction=&CameraDirectionVector}.Result
    B=ObjectOrientation()
   )
   B=ScalarParameter() {defaultValue=2, parameterName=Sheen_01_Bias}
  )
  n10 = TextureCoordinate()
  n11 = Constant() {r=3.8734}
  n12 = Multiply(
   A=ChannelMaskParameter(None=Rotator(Coordinate=@n10, Time=@n11)) {parameterName=SheenUVDirection}.R
   B=ScalarParameter() {defaultValue=4, parameterName=bandMultiplier}
  )
  n13 = ChannelMaskParameter(None=Frac(None=Add(A=@n9, B=@n12))) {maskChannel=Blue, defaultValue=<rgba 0 0 1 0>}
  n14 = Multiply(
   A=SphereMask(
    A=@n13.R
    B=ScalarParameter() {defaultValue=0.5, parameterName=Sheen_01_Pos, sliderMin=1}
    Radius=ScalarParameter() {defaultValue=0.05, parameterName=Sheen_01_Width}
    Hardness=ScalarParameter() {defaultValue=0.5, parameterName=Sheen_01_Hardness}
   )
   B=ScalarParameter() {parameterName=Sheen_01_Strength}
  )
  n15 = Multiply(
   A=SphereMask(
    A=@n13.R
    B=ScalarParameter() {defaultValue=0.3, parameterName=Sheen_02_Pos, sliderMin=1}
    Radius=ScalarParameter() {defaultValue=0.025, parameterName=Sheen_02_Width}
    Hardness=ScalarParameter() {defaultValue=1, parameterName=Sheen_02_Hardness}
   )
   B=ScalarParameter() {parameterName=Sheen_01_Strength}
  )
  n16 = Fresnel(ExponentIn=ScalarParameter() {defaultValue=1, parameterName=Exponent}, Normal=@n4)
  Add(
   A=Add(A=SphereMask(A=@n5, B=@n6, Radius=@n7, Hardness=@n8), B=Add(A=@n14, B=@n15))
   B=LinearInterpolate(A=ScalarParameter() {defaultValue=0.5, parameterName=MinOpacity}, B=ScalarParameter() {defaultValue=1, parameterName=MaxOpacity}, Alpha=@n16)
  )
  Constant()
  Constant() {r=1}
  n17 = Multiply(A=Subtract(A=PixelDepth().R, B=SceneDepth().R)) {constB=-1}
  n18 = Divide(
   A=Add(A=Clamp(None=@n17, Max=@n17), B=ScalarParameter() {defaultValue=-32, parameterName=DeptjhFadeDistance})
   B=ScalarParameter() {defaultValue=2, parameterName=DeptjhfadeBias}
  ) {constB=1}
  n19 = Clamp(None=OneMinus(None=Clamp(None=@n18)))
  Multiply(
   A=@n19
   B=LinearInterpolate(
    A=VectorParameter() {parameterName=Diffuse_02, defaultValue=<rgba 0 0 0 1>}.RGB
    B=VectorParameter() {parameterName=Diffuse, defaultValue=<rgba 0.2083 0.2382 0.2708 1>}.RGB
    Alpha=@n16
   )
  )
  Multiply(A=@n19, B=SceneColor!(UVs).RGB)
  TextureSampleParameter2D() {parameterName=Normal, samplerType=SAMPLERTYPE_Normal, texture=/Engine/EngineMaterials/FlatNormal}
  TextureSampleParameter2D() {parameterName=Diffuse, texture=&T_WhitesRGB}
 ]
}
